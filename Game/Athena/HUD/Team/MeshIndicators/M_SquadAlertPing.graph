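# Unlit, additive, two-sided material for the squad alert ping mesh indicator, apparently a glowing column in the team HUD.
# Two panning triangle textures are tinted by a dark-to-bright MainColor ramp along UV V, combined with BaseColor, then saturation-adjusted.
# An opacity term multiplies pattern plus ramp by a V-based power falloff.
# A Taper term scales vertex normals by inverted second-UV V to shape the mesh.
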Material /Game/Athena/HUD/Team/MeshIndicators/M_SquadAlertPing
graph {
 properties {
  bUsedWithNanite=true
  blendMode=BLEND_Additive
  pixelDepthOffsetMode=PDOM_Legacy
  shadingModel=MSM_Unlit
  twoSided=true
  usedShadingModels=Unlit
 }
 nodes [
  n1 = OneMinus(None=ComponentMask(None=TextureCoordinate() {coordinateIndex=1}) {g=true})
  n2 = VertexNormalWS()
  Add(A=Multiply(A=ScalarParameter() {defaultValue=-3.7, parameterName=Taper}, B=Multiply(A=@n1, B=@n2))) {constB=0}
  n3 = TextureCoordinate()
  n4 = Subtract(A=ComponentMask(None=@n3) {g=true}) {constB=0.25}
  n5 = TextureCoordinate()
  n6 = TextureSample(UVs=Multiply(A=Panner(Coordinate=@n5) {speedX=0.1, speedY=0.5})) {samplerSource=SSM_Wrap_WorldGroupSettings, texture=/Game/Global/Textures/Default/Masks_and_Gradients/T_TrianglePattern}
  n7 = TextureSample(UVs=Multiply(A=Panner(Coordinate=@n3) {speedX=0.3, speedY=0.08}, B=ScalarParameter() {defaultValue=0.75, parameterName=TriangleScale}) {constB=0.75}) {
   samplerSource=SSM_Wrap_WorldGroupSettings
   samplerType=SAMPLERTYPE_LinearColor
   texture=/Game/UI/Foundation/Textures/Pattern/T_Triangle_BattlePass
  }
  n8 = Add(A=Multiply(A=Power(Base=@n4), B=@n6.R), B=@n7.G)
  n9 = Multiply(
   A=VectorParameter() {parameterName=MainColor, defaultValue=<rgba 0.7168 0 1 1>}.RGB
   B=ScalarParameter() {defaultValue=1, parameterName=ColumnTintScalar}
  )
  n10 = ComponentMask(None=@n3) {g=true}
  n11 = Power(Base=@n10, Exp=ScalarParameter() {defaultValue=2.5446, parameterName=RampPower})
  n12 = LinearInterpolate(A=Multiply(A=ScalarParameter() {defaultValue=0.0817, parameterName=ColumnTintDarkScalar}, B=@n9), B=@n9, Alpha=@n11)
  n13 = Multiply(A=@n7.B, B=@n12)
  n14 = Add(A=Multiply(A=Multiply(A=@n8, B=@n12), B=@n7.G), B=Add(A=Multiply(A=@n6.G, B=@n12), B=Multiply(A=@n13, B=@n6.R)))
  n15 = Power(Base=@n11, Exp=ScalarParameter() {defaultValue=6.7523, parameterName=BaseColorPow})
  n16 = VectorParameter() {parameterName=BaseColor, defaultValue=<rgba 0.1009 0.5975 1 1>}
  n17 = Multiply(A=@n16.RGB, B=@n16.RGB)
  Desaturation(
   None=Add(A=Multiply(A=@n14, B=@n15), B=Multiply(A=@n15, B=@n17))
   Fraction=ScalarParameter() {defaultValue=-0.8047, parameterName=SaturationAdjust}
  )
  LinearInterpolate() {constA=1, constB=0.75}
  LinearInterpolate() {constA=0.5, constB=2.5}
  Multiply(
   A=Multiply(A=Add(A=@n6.R, B=@n11), B=ScalarParameter() {defaultValue=0.5, parameterName=overallopacity})
   B=Power(Base=@n10, Exp=ScalarParameter() {defaultValue=0.7744, parameterName=OpacityPow})
  )
  TextureCoordinate() {coordinateIndex=1}
  TextureCoordinate() {coordinateIndex=1}
 ]
}
