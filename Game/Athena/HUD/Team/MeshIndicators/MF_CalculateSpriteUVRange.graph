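# Material function in the team HUD mesh indicators folder that computes the normalized UV range of one sprite in a padded sprite-sheet texture.
# A single Custom HLSL node takes texture size, sprite size, padding and sprite index, deriving sprites per row from the padded cell width.
# It converts the index to a row and column, then returns a float4 packed as U start, U end, V start, V end.
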MaterialFunction /Game/Athena/HUD/Team/MeshIndicators/MF_CalculateSpriteUVRange
graph {nodes [
  MaterialExpressionFunctionInput_0 = FunctionInput() {inputName=Texture Size, inputType=FunctionInput_Vector2, previewValue=<1024 1024 0 0>}
  MaterialExpressionFunctionInput_1 = FunctionInput() {inputName=Sprite Size, inputType=FunctionInput_Vector2, sortPriority=1, previewValue=<128 128 0 1>}
  MaterialExpressionFunctionInput_2 = FunctionInput() {inputName=Padding, inputType=FunctionInput_Scalar, sortPriority=2, previewValue=<8 0 0 1>}
  MaterialExpressionFunctionInput_5 = FunctionInput() {inputName=Sprite Index, inputType=FunctionInput_Scalar, sortPriority=3}
  n1 = Custom!(
   TextureSize=@MaterialExpressionFunctionInput_0
   SpriteSize=@MaterialExpressionFunctionInput_1
   Padding=@MaterialExpressionFunctionInput_2
   SpriteIndex=@MaterialExpressionFunctionInput_5
  ) {
   code="float2 SpriteTotalSize = SpriteSize + Padding*2;\r\nint SpritesPerRow = floor(TextureSize.x / SpriteTotalSize.x);\r\n\r\nint ColumnIndex = SpriteIndex % SpritesPerRow;\r\nint RowIndex = floor(SpriteIndex / SpritesPerRow);\r\n\r\nfloat PixelX = ColumnIndex * SpriteTotalSize.x + Padding;\r\nfloat PixelY = RowIndex * SpriteTotalSize.y + Padding;\r\n\r\nfloat PixelX_End = PixelX + SpriteSize.x;\r\nfloat PixelY_End = PixelY + SpriteSize.y;\r\n\r\nreturn float4(PixelX, PixelX_End, PixelY, PixelY_End) / float4(TextureSize.x, TextureSize.x, TextureSize.y, TextureSize.y);"
   outputType=CMOT_Float4
   inputs [
    {inputName=TextureSize, input {expression=&self:MaterialExpressionFunctionInput_0}}
    {inputName=SpriteSize, input {expression=&self:MaterialExpressionFunctionInput_1}}
    {inputName=Padding, input {expression=&self:MaterialExpressionFunctionInput_2}}
    {inputName=SpriteIndex, input {expression=&self:MaterialExpressionFunctionInput_5}}
   ]
  }
  FunctionOutput(None=@n1)
 ]}
