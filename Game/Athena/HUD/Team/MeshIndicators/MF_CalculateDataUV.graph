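# Material function for the team HUD mesh indicators that converts a column and row into a UV for a data atlas texture.
# It combines Column and Row into a float2, divides by Data Atlas Size, then adds a half-texel offset derived from that size.
# The resulting UV lands on texel centers, assuming integer cell indices and an atlas size given in texels.
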MaterialFunction /Game/Athena/HUD/Team/MeshIndicators/MF_CalculateDataUV
graph {nodes [
  n1 = MaterialFunctionCall(X=FunctionInput() {inputName=Column, inputType=FunctionInput_Scalar}, Y=FunctionInput() {inputName=Row, inputType=FunctionInput_Scalar}) {materialFunction=&MakeFloat2}
  n2 = FunctionInput() {inputName=Data Atlas Size, inputType=FunctionInput_Vector2}
  n3 = Multiply(A=@n2) {constB=2}
  FunctionOutput(None=Add(A=Divide(A=@n1.Result, B=@n2), B=Divide(B=@n3))) {outputName=UV + HalfTexelOffset}
 ]}
