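# Material function that decodes one row of a team data atlas for HUD team mesh indicators.
# MF_CalculateDataUV turns the atlas size, Row input and a fixed column into a texel-centered UV, sampled at an explicit mip level.
# Red, green and alpha are scaled from normalized to byte range and rounded to output Rank, MemberState and X-OriginOffset; blue passes through as Opacity.
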MaterialFunction /Game/Athena/HUD/Team/MeshIndicators/MF_ReadTeamAtlas
graph {nodes [
  n1 = TextureSample!(
   UVs=MaterialFunctionCall!(
    Data Atlas Size=FunctionInput() {inputName=Data Atlas Size, inputType=FunctionInput_Vector2}
    Column=Constant()
    Row=FunctionInput() {inputName=Row, inputType=FunctionInput_Scalar}
   ) {materialFunction=&/MF_CalculateDataUV}.UV + HalfTexelOffset
   Tex=FunctionInput(Preview=TextureObject() {samplerType=SAMPLERTYPE_LinearColor, texture=&/TestAtlas}) {bUsePreviewValueAsDefault=true, inputName=Data Atlas, inputType=FunctionInput_Texture2D}
   Level
   Apply View MipBias
  ) {automaticViewMipBias=false, mipValueMode=TMVM_MipLevel, samplerType=SAMPLERTYPE_LinearColor}
  FunctionOutput(None=Round(None=Multiply(A=@n1.R) {constB=255})) {outputName=Rank}
  FunctionOutput(None=Round(None=Multiply(A=@n1.G) {constB=255})) {outputName=MemberState, sortPriority=1}
  FunctionOutput(None=Round(None=Multiply(A=@n1.A) {constB=255})) {outputName=X-OriginOffset, sortPriority=3}
  FunctionOutput(None=@n1.B) {outputName=Opacity, sortPriority=2}
 ]}
