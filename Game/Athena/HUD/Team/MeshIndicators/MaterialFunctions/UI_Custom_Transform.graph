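# Material function that applies a custom scale-and-translate transform to 2D vertex positions for the team HUD mesh indicators.
# It adds an Offset input to a VertexPosition input, which previews as world XY excluding shader offsets.
# The sum is multiplied by UI_Custom_UnpackScale, routed through a ShadingPathSwitch with only Default wired, then shifted by UI_Custom_UnpackPosition.
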MaterialFunction /Game/Athena/HUD/Team/MeshIndicators/MaterialFunctions/UI_Custom_Transform
graph {nodes [
  Constant() {r=1}
  n1 = WorldPosition() {worldPositionShaderOffset=WPT_ExcludeAllShaderOffsets}
  n2 = Add(
   A=FunctionInput() {bUsePreviewValueAsDefault=true, inputName=Offset, inputType=FunctionInput_Vector2}
   B=FunctionInput(Preview=ComponentMask(None=@n1.XYZ) {g=true, r=true}) {bUsePreviewValueAsDefault=true, inputName=VertexPosition, inputType=FunctionInput_Vector2}
  )
  n3 = ShadingPathSwitch(Default=MaterialFunctionCall() {materialFunction=&/UI_Custom_UnpackScale}.Result) {desc=Migrated from FeatureLevelSwitch}
  FunctionOutput(None=Add(
    A=Multiply(A=@n2, B=@n3)
    B=MaterialFunctionCall() {materialFunction=&/UI_Custom_UnpackPosition}.Result
   ))
 ]}
