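# Translucent UI material for team indicators in the HUD, drawing each member's state icon from a sprite atlas.
# An unpacked base address selects a row in the DataAtlas texture parameter, which MF_ReadTeamAtlas decodes into member state, X origin offset and opacity.
# State picks the sprite's UV range, opacity scales sprite alpha, and the offset feeds a custom UI transform for horizontal placement.
# A GetDimensions custom node and TestIndex parameter sit unconnected.
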Material /Game/Athena/HUD/Team/MeshIndicators/M_TeamIndicator
graph {
 properties {blendMode=BLEND_Translucent, materialDomain=MD_UI, pixelDepthOffsetMode=PDOM_Legacy, usedShadingModels=Default Lit}
 nodes [
  n1 = Constant2Vector() {g=512, r=1024}
  n2 = Constant2Vector() {g=128, r=128}
  n3 = Constant()
  n4 = TextureObjectParameter() {group=Data, parameterName=DataAtlas, samplerType=SAMPLERTYPE_LinearColor, texture=&/TestAtlas}
  n5 = VectorParameter() {group=Data, parameterName=DataAtlasSize, defaultValue=<rgba 0 0 0 1>}
  n6 = MaterialFunctionCall!(
   Data Atlas=@n4
   Data Atlas Size=MaterialFunctionCall(X=@n5.R, Y=@n5.R) {materialFunction=&MakeFloat2}.Result
   Row=MaterialFunctionCall() {materialFunction=/Game/Athena/HUD/Team/MeshIndicators/MaterialFunctions/UI_Custom_UnpackBaseAddress}.Result
  ) {materialFunction=&/MF_ReadTeamAtlas}
  n7 = TextureSample(UVs=MaterialFunctionCall!(
    X range Y range=MaterialFunctionCall!(Texture Size=@n1, Sprite Size=@n2, Padding=@n3, Sprite Index=@n6.MemberState) {materialFunction=&/MF_CalculateSpriteUVRange}.Result
    UVs=TextureCoordinate()
   ) {materialFunction=/Engine/Functions/Engine_MaterialFunctions01/Coordinates/UVRemap_0-1_ToRange}.UV) {texture=&/T_TeamIndicatorAtlas_2}
  ComponentMask(None=@n7.RGB) {b=true, g=true, r=true}
  Custom!(Tex) {
   code="float2 SurfaceDimensions;\r\nTextureObject.GetDimensions(SurfaceDimensions.x,SurfaceDimensions.y);\r\n\r\nreturn SurfaceDimensions;"
   description=GetDimensions2D
   outputType=CMOT_Float2
   inputs [{inputName=TextureObject, input {expression=null}}]
  }
  MaterialFunctionCall!(
   VertexPosition
   Offset=MaterialFunctionCall(X=@n6.X-OriginOffset, Y=Constant()) {materialFunction=&MakeFloat2}.Result
  ) {materialFunction=/Game/Athena/HUD/Team/MeshIndicators/MaterialFunctions/UI_Custom_Transform}
  Multiply(A=@n7.A, B=@n6.Opacity) {constA=1}
  ScalarParameter() {defaultValue=2, parameterName=TestIndex}
  TextureProperty()
 ]
}
